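# Editor utility widget that serves as a reusable custom button for the BookShelf tool, built from a Border wrapping a named slot for child content.
# PreConstruct applies the exposed brush, color and slot padding variables to the border.
# Mouse down fires Clicked, or CTRL Clicked when Control is held, and double-click fires Double Clicked.
# Both handlers pass the pointer event to their dispatcher and return Unhandled.
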EditorUtilityWidgetBlueprint /Game/NeverCook/TechArt/BookShelf/Dependencies/Widgets/EUW_BookCase_Button
graphs [
 CTRL Clicked {nodes [CTRL Clicked!() -> (then, Event:Pointer Event Structure) {is=title K2Node_FunctionEntry}]}
 Clicked {nodes [Clicked!() -> (then, Event:Pointer Event Structure) {is=title K2Node_FunctionEntry}]}
 Double Clicked {nodes [Double Clicked!() -> (then, Event:Pointer Event Structure) {is=title K2Node_FunctionEntry}]}
 EventGraph {nodes [
   Event PreConstruct()
   n1 = Get Button() -> (Button:Border Object Reference)
   > SetBrush(self:Border Object Reference=@n1, InBrush=Get Brush() -> (Brush:Slate Brush Structure))
   > SetBrushColor(self=@n1, InBrushColor=Get Brush Color() -> (Brush Color:Linear Color Structure))
   > SetPadding!(
    execute
    self:Border Slot Object Reference=SlotAsBorderSlot!(Widget:Widget Object Reference=Get Child() -> (Child:Named Slot Object Reference)) -> (ReturnValue:Border Slot Object Reference)
    InPadding:Margin Structure=Get Button Padding() -> (Button Padding:Margin Structure)
   ) -> (then)
   Event Construct()
   Event Tick()
  ]}
 On_Button_MouseButtonDown {nodes [
   n1 = On_Button_MouseButtonDown!() -> (then, MyGeometry:Geometry Structure, MouseEvent:"Pointer Event Structure (by ref)") {is=title K2Node_FunctionEntry}
   > n2 = Branch(Condition=Is Control Down!(Input:"Input Event Structure (by ref)"=GetInputEventFromPointerEvent!(Event:"Pointer Event Structure (by ref)"=@n1.MouseEvent) -> (ReturnValue:Input Event Structure)) -> (ReturnValue:bool))
   > n3 = Call CTRL Clicked!(execute:Exec, self:EUW Book Case Button Object Reference, Event:Pointer Event Structure=@n1.MouseEvent) -> (then:Exec) {is=title K2Node_CallDelegate}
   Return Node!(execute=[@n4.then, @n3.then], ReturnValue:Event Reply Structure=Unhandled())
   n4 = Call Clicked!(execute:Exec=@n2.else, self:EUW Book Case Button Object Reference, Event:Pointer Event Structure=@n1.MouseEvent) -> (then:Exec) {is=title K2Node_CallDelegate}
  ]}
 On_Button_MouseDoubleClick {nodes [
   n1 = On_Button_MouseDoubleClick!() -> (then, MyGeometry:Geometry Structure, MouseEvent:"Pointer Event Structure (by ref)") {is=title K2Node_FunctionEntry}
   > Call Double Clicked!(execute:Exec, self:EUW Book Case Button Object Reference, Event:Pointer Event Structure=@n1.MouseEvent) -> (then:Exec) {is=title K2Node_CallDelegate}
   > Return Node(ReturnValue=Unhandled())
  ]}
]
variables [
 Brush {
  type=SlateBrush
  default {
   bIsDynamicallyLoaded=false
   drawAs=RoundedBox
   imageType=NoImage
   mirroring=NoMirror
   resourceName=None
   resourceObject=null
   tiling=NoTile
   imageSize=<32 32>
   margin=<ltrb 0 0 0 0>
   outlineSettings {
    bUseBrushTransparency=false
    roundingType=HalfHeightRadius
    width=0
    color {colorUseRule=UseColor_Specified, specifiedColor=<rgba 0 0 0 0>}
    cornerRadii=<2 2 2 2>
   }
   tintColor {colorUseRule=UseColor_Specified, specifiedColor=<rgba 1 1 1 1>}
   uVRegion {bIsValid=false, max=<0 0>, min=<0 0>}
  }
 }
 Brush Color {type=LinearColor, default=<rgba 0 0 0 0.8>}
 Button Padding {type=Margin, default=<ltrb 0 0 0 0>}
 Clicked {default=()}
 CTRL Clicked {default=()}
 Double Clicked {default=()}
]
widgets {
 info {rootWidgetClass=&Border, widgetCount=2}
 widgets [
  Button {widgetClassPath=&Border, bIsVariable=true, widget=Button, properties {padding=<ltrb 1 1 1 1>}}
  Child {
   widgetClassPath=&NamedSlot
   bIsVariable=true
   parent=Button
   widget=Child
   slot {class=BorderSlot, properties {padding=<ltrb 1 1 1 1>}}
  }
 ]
}
